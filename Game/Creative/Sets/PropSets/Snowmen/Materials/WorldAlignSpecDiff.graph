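# Default lit material for Creative snowman props that projects snow diffuse and mask textures in world space, with shared tile size and transition contrast.
# The mask's green channel drives Metallic, Specular and Roughness alike.
# A UV-mapped snowman normal blends with a tiled, softened battered-steel detail normal through BlendAngleCorrectedNormals.
# The resulting material attributes pass through MF_CreativeMoveTool before output.
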Material /Game/Creative/Sets/PropSets/Snowmen/Materials/WorldAlignSpecDiff
graph {
 properties {bUseMaterialAttributes=true, pixelDepthOffsetMode=PDOM_Legacy, usedShadingModels=Default Lit}
 nodes [
  n1 = TextureObjectParameter() {parameterName=Diffuse, texture=/Game/Athena/WinterArt/Textures/T_Snow01_D}
  n2 = VectorParameter() {parameterName=WorldTileScale, defaultValue=<rgba 128 128 128 0>}
  n3 = ScalarParameter() {defaultValue=0.5, parameterName="ProjectionTransitionContrast (S)"}
  n4 = MaterialFunctionCall(Float3=MaterialFunctionCall(
    TextureObject=TextureObjectParameter() {parameterName=SpecularMasks, samplerType=SAMPLERTYPE_Masks, texture=/Game/Athena/WinterArt/Textures/T_Snow01_M}
    TextureSize=@n2.RGB
    ProjectionTransitionContrast=@n3
   ) {materialFunction=&WorldAlignedTexture}.XYZ Texture) {materialFunction=&BreakOutFloat3Components}
  n5 = TextureSampleParameter2D() {parameterName=Normals, samplerType=SAMPLERTYPE_Normal, texture=/Game/Creative/Sets/PropSets/Snowmen/Textures/T_Snowman_Kit_03_N}
  n6 = TextureCoordinate()
  n7 = ScalarParameter() {defaultValue=5, parameterName=DetailNormalTiling}
  n8 = Multiply(
   A=TextureSample(UVs=Multiply(A=@n6, B=@n7)) {samplerType=SAMPLERTYPE_Normal, texture=/Game/Characters/Player/Male/Medium/Bodies/M_MED_Doggus/Textures/T_BatteredSteel_N}.RGB
   B=Constant3Vector() {constant=<rgba 0.5 0.5 1 0>}
  )
  MaterialFunctionCall(MA In=MakeMaterialAttributes!(
    BaseColor=MaterialFunctionCall(TextureObject=@n1, TextureSize=@n2.RGB, ProjectionTransitionContrast=@n3) {materialFunction=&WorldAlignedTexture}.XYZ Texture
    Metallic=@n4.G
    Specular=@n4.G
    Roughness=@n4.G
    Anisotropy
    EmissiveColor
    Opacity
    OpacityMask
    Normal=MaterialFunctionCall(BaseNormal=@n5.RGB, AdditionalNormal=@n8) {materialFunction=&BlendAngleCorrectedNormals}.Result
    Tangent
    WorldPositionOffset
    SubsurfaceColor
    ClearCoat
    ClearCoatRoughness
    AmbientOcclusion
    Refraction
    CustomizedUVs_0
    CustomizedUVs_1
    CustomizedUVs_2
    CustomizedUVs_3
    CustomizedUVs_4
    CustomizedUVs_5
    CustomizedUVs_6
    CustomizedUVs_7
    PixelDepthOffset
    ShadingModel
    Displacement
   )) {materialFunction=&MF_CreativeMoveTool}
 ]
}
